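# Translucent lit particle material for frosty wisps in a skydiving ice trail.
# Color adds panning tiled dust to an 8x8 SubUV mist flipbook's red channel, tinted by particle color, with switchable dynamic-parameter control of emissive and density.
# Emissive branch adds a tweak tint and optional time-of-day scaling.
# Opacity combines flipbook blue and particle alpha, with near-camera and depth fades that ES2 and ES3.1 skip.
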Material /Figment_S10/Athena/Effects/Materials/Skydiving/M_IceTrail_FrostyWisps
graph {
 properties {
  bUsedWithNiagaraSprites=true
  bUsedWithParticleSprites=true
  blendMode=BLEND_Translucent
  pixelDepthOffsetMode=PDOM_Legacy
  usedShadingModels=Default Lit
 }
 nodes [
  n1 = ParticleSubUV!(UVs, Tex, Apply View MipBias) {samplerType=SAMPLERTYPE_LinearColor, texture=/Game/Effects/Fort_Effects/Textures/Smoke/T_FallingMist_8x8_01}
  n2 = Multiply(A=@n1.B, B=ParticleColor().A)
  n3 = ScalarParameter() {defaultValue=500, parameterName="Distance (S)"}
  n4 = DepthFade(
   Opacity=Multiply(A=@n2, B=MaterialFunctionCall(Distance=@n3) {materialFunction=&Near_Camera_Fade}.Result)
   FadeDistance=ScalarParameter() {defaultValue=25, parameterName=depth fade}
  )
  FeatureLevelSwitch(Default=@n4, ES2=@n2, ES3_1=@n2, SM4_REMOVED=@n4, SM5=@n4, SM6=@n4)
  n5 = DynamicParameter() {defaultValue=<rgba 1 1 1 0>, paramNames [DustTiling, Dust Emissive, Cloud Density, Panning MacroUV Control]}
  n6 = StaticSwitchParameter(True=ParticleMacroUV(), False=TextureCoordinate()) {parameterName=USe MacroUV}
  n7 = DynamicParameter() {defaultValue=<rgba 1 1 1 0>, paramNames [DustTiling, Dust Emissive, Cloud Density, Panning MacroUV Control]}
  n8 = Time()
  n9 = ScalarParameter() {parameterName=U_Speed}
  n10 = ScalarParameter() {parameterName=V_Speed}
  n11 = TextureSample(UVs=Panner(Coordinate=Multiply(A=@n6, B=@n7.DustTiling), Time=Multiply(A=@n7.Panning MacroUV Control, B=@n8), Speed=AppendVector(A=@n9, B=@n10))) {texture=/Game/Effects/Fort_Effects/Textures/Tile/T_Tiling_Dust}
  n12 = Multiply(A=@n5.Dust Emissive, B=@n11.RGB)
  n13 = Multiply(A=@n5.Cloud Density, B=@n1.R)
  n14 = Multiply(
   A=ParticleColor().RGB
   B=Add(
    A=StaticSwitchParameter(True=@n12, False=@n11.RGB) {parameterName=Dynamic Control of Dust Emissive}
    B=StaticSwitchParameter(True=@n13, False=@n1.R) {parameterName=Dynamic Control of Cloud Density}
   )
  )
  Reroute(None=Multiply(A=ScalarParameter() {defaultValue=1, parameterName=BaseColorIntensity}, B=@n14))
  n15 = ScalarParameter() {defaultValue=1.5, parameterName=day color}
  n16 = ScalarParameter() {defaultValue=1, parameterName=night color}
  n17 = Multiply(A=@n14, B=VectorParameter() {parameterName=EmissiveTweak, defaultValue=<rgba 2 2 2 0>}.RGB) {constB=0.5}
  StaticSwitchParameter(
   True=Multiply(
    A=MaterialFunctionCall(Color_Morning=@n15, Color_Day=@n15, Color_Evening=@n16, Color_Night=@n16) {materialFunction=&Color_Time_of_Day}.Result
    B=@n17
   )
   False=@n17
  ) {parameterName=color time of day}
 ]
}
